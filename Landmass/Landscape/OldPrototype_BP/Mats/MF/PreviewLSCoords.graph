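# Material function from the old Landmass landscape prototype that outputs 2D preview coordinates over a landscape area centred on the actor.
# World XY relative to the actor position is divided by half the total size from MPC_Erosion, then bias-scaled to a 0–1 range.
# The TexelScale parameter blends in a Res/(Res−1) size correction, using that collection's resolution, to account for the one-texel difference.
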MaterialFunction /Landmass/Landscape/OldPrototype_BP/Mats/MF/PreviewLSCoords
graph {nodes [
  n1 = ComponentMask(None=Subtract(A=WorldPosition().XYZ, B=ActorPositionWS())) {g=true, r=true}
  n2 = CollectionParameter() {collection=/Landmass/Landscape/BlueprintBrushes/MPC/MPC_Erosion, parameterName=TotalSizeX}
  n3 = CollectionParameter() {collection=&MPC_Erosion, parameterName=TotalSizeY}
  n4 = AppendVector(
   A=CollectionParameter() {collection=&MPC_Erosion, parameterName=ResX}
   B=CollectionParameter() {collection=&MPC_Erosion, parameterName=ResY}
  )
  n5 = Divide(A=@n4, B=Subtract(A=@n4))
  n6 = ScalarParameter() {defaultValue=1, parameterName=TexelScale}
  n7 = Divide(A=Multiply(A=AppendVector(A=@n2, B=@n3), B=LinearInterpolate(B=@n5, Alpha=@n6) {constA=1}))
  FunctionOutput(None=ConstantBiasScale(None=Divide(A=@n1, B=@n7)))
 ]}
