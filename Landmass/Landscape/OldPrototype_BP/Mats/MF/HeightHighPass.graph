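# HeightHighPass is a prototype Landmass material function that derives a cavity or curvature shading mask from a landscape heightmap.
# It runs a high-quality HighPassTexture on the HeightRT parameter, using PreviewLSCoords UVs and an offset parameter.
# The no-contrast red channel is normalized by Curvature Z Max, biased, shaped by a contrast power and saturated.
# That mask blends up from Cavity Min, then Final Mult scales the output.
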MaterialFunction /Landmass/Landscape/OldPrototype_BP/Mats/MF/HeightHighPass
graph {nodes [
  n1 = ScalarParameter() {defaultValue=0.4, parameterName=Cavity Min}
  n2 = MaterialFunctionCall!(
   Texture Object=TextureObjectParameter() {parameterName=HeightRT, texture=&Black}
   UVs=MaterialFunctionCall() {materialFunction=&/PreviewLSCoords}.Result
   Contrast
   Highpass Offset=ScalarParameter() {defaultValue=16, parameterName=O}
   High Quality=StaticBool() {value=true}
  ) {materialFunction=/Engine/Functions/Engine_MaterialFunctions01/ImageAdjustment/HighPassTexture}
  n3 = Add(A=Divide(A=ComponentMask(None=@n2.Result No Contrast) {r=true}, B=ScalarParameter() {defaultValue=20000, parameterName=Curvature Z Max}))
  n4 = ScalarParameter() {defaultValue=6, parameterName=Curvature Contrast}
  n5 = Saturate(None=Multiply(A=Power(Base=@n3, Exp=@n4)) {constB=0.5})
  FunctionOutput(None=Multiply(A=LinearInterpolate(A=@n1, Alpha=@n5), B=ScalarParameter() {defaultValue=1.5, parameterName=Final Mult}))
 ]}
