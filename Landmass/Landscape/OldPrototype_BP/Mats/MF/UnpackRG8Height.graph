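# Material function in the Landmass old prototype that decodes a height value packed into two 8-bit channels.
# It splits the RG input, rescales both channels by 255, and combines R as the high byte with G as the low byte.
# It then subtracts the 32768 midpoint, outputting a signed height in raw 16-bit units with no world-space scale applied.
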MaterialFunction /Landmass/Landscape/OldPrototype_BP/Mats/MF/UnpackRG8Height
graph {nodes [
  n1 = MaterialFunctionCall(Float2=FunctionInput() {inputName=RG, inputType=FunctionInput_Vector2}) {materialFunction=&BreakOutFloat2Components}
  n2 = Multiply(A=Multiply(A=@n1.R) {constB=255}) {constB=256}
  n3 = Multiply(A=@n1.G) {constB=255}
  FunctionOutput(None=Subtract(A=Add(A=@n2, B=@n3)) {constB=32768})
 ]}
