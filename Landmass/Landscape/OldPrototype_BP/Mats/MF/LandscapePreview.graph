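# Material function from the Landmass landscape prototype that previews procedural height edits on a landscape through vertex offset and normals.
# Landscape coordinates divided by the MPC_Erosion resolution sample three render target parameters.
# StackRT height minus unpacked RG8 LandscapeRT height drives Z-only WPO, scaled by WPO and ShowProcedural.
# NormalRT's RG rebuilds a normal, replacing the vertex normal where its alpha is below MinDetectionThreshold, also weighted by ShowProcedural.
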MaterialFunction /Landmass/Landscape/OldPrototype_BP/Mats/MF/LandscapePreview
graph {nodes [
  n1 = Divide(
   A=LandscapeLayerCoords() {mappingScale=1}
   B=AppendVector(
    A=CollectionParameter() {collection=/Landmass/Landscape/BlueprintBrushes/MPC/MPC_Erosion, parameterName=ResX}
    B=CollectionParameter() {collection=&MPC_Erosion, parameterName=ResY}
   )
  )
  n2 = TextureSampleParameter2D(UVs=@n1) {parameterName=LandscapeRT, samplerType=SAMPLERTYPE_LinearColor, texture=/Landmass/Landscape/OldPrototype_BP/Textures/T_FlatLS_RG8}
  n3 = Multiply(
   A=Subtract(
    A=TextureSampleParameter2D(UVs=@n1) {parameterName=StackRT, texture=&Black}.R
    B=MaterialFunctionCall!(RG=@n2.RGB) {materialFunction=&/UnpackRG8Height}.Result
   )
   B=Constant3Vector() {constant=<rgba 0 0 1 0>}
  )
  n4 = ScalarParameter() {defaultValue=1, parameterName=WPO}
  n5 = CollectionParameter() {collection=&MPC_Erosion, parameterName=ShowProcedural}
  FunctionOutput(None=Multiply(A=Multiply(A=@n3, B=@n4), B=@n5)) {outputName=WPO, sortPriority=1}
  n6 = TextureSampleParameter2D(UVs=@n1) {parameterName=NormalRT, samplerType=SAMPLERTYPE_LinearColor, texture=/Engine/ArtTools/RenderToTexture/Textures/127grey}
  n7 = ConstantBiasScale(None=ComponentMask(None=@n6.RGB) {g=true, r=true}) {bias=-0.5, scale=2}
  n8 = Constant()
  n9 = If(A=@n6.A, B=ScalarParameter() {defaultValue=0.999, parameterName=MinDetectionThreshold}, "A > B"=@n8, "A == B"=@n8, "A < B"=Constant() {r=1})
  FunctionOutput(None=LinearInterpolate(A=VertexNormalWS(), B=DeriveNormalZ(InXY=@n7), Alpha=Multiply(A=@n5, B=@n9))) {outputName=NormalOffset}
 ]}
