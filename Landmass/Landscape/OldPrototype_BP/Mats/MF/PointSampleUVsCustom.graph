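# Material function in the Landmass old prototype that converts UVs into texel-center UVs for point sampling a texture.
# It multiplies the UVs by a Texture Size input, floors the result, adds a half-texel offset, then divides by the size again.
# UVs default to the mesh texture coordinates, and Texture Size has a preview default for when it is left unconnected.
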MaterialFunction /Landmass/Landscape/OldPrototype_BP/Mats/MF/PointSampleUVsCustom
graph {nodes [
  n1 = TextureCoordinate()
  n2 = FunctionInput() {bUsePreviewValueAsDefault=true, inputName=Texture Size, inputType=FunctionInput_Vector2, sortPriority=1, previewValue=<256 256 0 1>}
  n3 = Floor(None=Multiply(A=FunctionInput(Preview=@n1) {bUsePreviewValueAsDefault=true, inputName=UVs, inputType=FunctionInput_Vector2}, B=@n2))
  FunctionOutput(None=Divide(A=Add(A=@n3) {constB=0.5}, B=@n2)) {outputName=TexelCenterUVs}
 ]}
