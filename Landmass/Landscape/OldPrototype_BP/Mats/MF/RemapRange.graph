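# RemapRange is a material function in the Landmass old prototype that linearly remaps an input from a chosen range to 0–1.
# It subtracts the 0 Point from Input, divides by the span between the 1 Point and 0 Point, and saturates the result.
# The span is floored at a small positive value, which avoids division by zero but does not support inverted ranges.
# Both range points are optional inputs with preview values as defaults.
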MaterialFunction /Landmass/Landscape/OldPrototype_BP/Mats/MF/RemapRange
graph {nodes [
  n1 = FunctionInput() {bUsePreviewValueAsDefault=true, inputName=0 Point, sortPriority=1}
  n2 = Subtract(A=FunctionInput() {inputName=Input}, B=@n1)
  n3 = Max(A=Subtract(A=FunctionInput() {bUsePreviewValueAsDefault=true, inputName=1 Point, sortPriority=2, previewValue=<1 1 1 1>}, B=@n1)) {constB=0.001}
  FunctionOutput(None=Saturate(None=Divide(A=@n2, B=@n3)))
 ]}
