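# Material function for triplanar world-space tiling of a normal map, from the Landmass landscape prototype folder.
# World position over TextureSize drives three planar samples, with channels rebuilt using signs and components of the WS Normal input.
# Contrast-adjusted absolute normal X blends the side projections, and Z blends in the top.
# Normalized outputs provide the full blend, sides only, top only, and a flat-top variant using the input normal.
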MaterialFunction /Landmass/Landscape/OldPrototype_BP/Mats/MF/WorldAlignedNormal_Landmass
graph {
 properties {description=Tiles a texture in worldspace, libraryCategoriesText [Misc,Texturing]}
 nodes [
  Constant()
  n1 = FunctionInput() {bUsePreviewValueAsDefault=true, description=Size of texture in worldspace units, inputName=TextureSize, previewValue=<64 64 64 6>}
  n2 = Divide(
   A=FunctionInput(Preview=WorldPosition() {worldPositionShaderOffset=WPT_ExcludeAllShaderOffsets}.XYZ) {bUsePreviewValueAsDefault=true, inputName=WorldPosition, sortPriority=100}
   B=@n1
  )
  n3 = FunctionInput(Preview=TextureObject() {samplerType=SAMPLERTYPE_LinearColor, texture=/Engine/EngineMaterials/DefaultNormal_Uncompressed}) {inputName=TextureObject, inputType=FunctionInput_Texture2D}
  n4 = TextureSample(UVs=ComponentMask(None=@n2) {b=true, r=true}, Tex=@n3) {samplerSource=SSM_Wrap_WorldGroupSettings, samplerType=SAMPLERTYPE_Normal}
  n5 = Multiply(A=@n4.G) {constB=-1}
  n6 = FunctionInput(Preview=VertexNormalWS()) {bUsePreviewValueAsDefault=true, inputName=WS Normal, previewValue=<0 0 1 1>}
  n7 = Multiply(A=@n6, B=Constant() {r=1})
  n8 = ComponentMask(None=@n7) {g=true}
  n9 = TextureSample(UVs=ComponentMask(None=@n2) {b=true, g=true}, Tex=@n3) {samplerSource=SSM_Wrap_WorldGroupSettings, samplerType=SAMPLERTYPE_Normal}
  n10 = Multiply(A=@n9.G) {constB=-1}
  n11 = ComponentMask(None=@n7) {r=true}
  n12 = FunctionInput() {
   bUsePreviewValueAsDefault=true
   inputName=ProjectionTransitionContrast
   inputType=FunctionInput_Scalar
   sortPriority=10
   previewValue=<1 0 0 1>
  }
  n13 = MaterialFunctionCall(In=Abs(None=@n11), Contrast=@n12) {materialFunction=&CheapContrast}
  n14 = LinearInterpolate(
   A=MaterialFunctionCall(X=Multiply(A=@n4.R), Y=Multiply(A=@n5, B=@n8), Z=@n4.B) {materialFunction=&MakeFloat3}.Result
   B=MaterialFunctionCall(X=Multiply(A=@n10, B=@n11), Y=Multiply(A=@n9.R), Z=@n9.B) {materialFunction=&MakeFloat3}.Result
   Alpha=@n13.Result
  )
  n15 = TextureSample(UVs=ComponentMask(None=@n2) {g=true, r=true}, Tex=@n3) {samplerSource=SSM_Wrap_WorldGroupSettings, samplerType=SAMPLERTYPE_Normal}
  n16 = ComponentMask(None=@n7) {b=true}
  n17 = MaterialFunctionCall(In=Abs(None=@n16), Contrast=@n12) {materialFunction=&CheapContrast}
  FunctionOutput(None=Normalize(VectorInput=LinearInterpolate(A=@n14, B=@n15.RGB, Alpha=@n17.Result))) {outputName=XYZ Texture, sortPriority=1}
  FunctionOutput(None=Normalize(VectorInput=@n14)) {outputName=XY Texture, sortPriority=1}
  n18 = ComponentMask(None=Multiply(A=@n15.RGB, B=Constant3Vector() {constant=<rgba 1 1 1 0>})) {g=true, r=true}
  n19 = Multiply(A=@n15.B, B=@n16)
  FunctionOutput(None=Normalize(VectorInput=AppendVector(A=@n18, B=@n19))) {outputName=Z Texture, sortPriority=1}
  FunctionOutput(None=Normalize(VectorInput=LinearInterpolate(A=@n14, B=@n6, Alpha=@n17.Result))) {outputName=XYZFlatTop, sortPriority=1}
  OneMinus(None=@n13.Result)
  OneMinus(None=@n17.Result)
  Transform()
 ]
}
