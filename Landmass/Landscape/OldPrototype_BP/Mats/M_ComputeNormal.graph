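# Unlit prototype Landmass material that computes terrain normals from a height render target, with sizing and scale from MPC_Erosion.
# A Custom node snaps UVs to texel centers, samples four neighbors and packs two normals' XY into a float4: height alone (R) and height plus water (B).
# OutputLDR lerps between raw and bias-scaled results.
# A second forward-difference normal node and several stray nodes appear to be leftovers.
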Material /Landmass/Landscape/OldPrototype_BP/Mats/M_ComputeNormal
graph {
 properties {
  bAllowNegativeEmissiveColor=true
  blendMode=BLEND_AlphaComposite
  pixelDepthOffsetMode=PDOM_Legacy
  shadingModel=MSM_Unlit
  usedShadingModels=Unlit
 }
 nodes [
  MaterialExpressionTextureObjectParameter_1 = TextureObjectParameter() {parameterName=HeightRT, texture=&Black}
  MaterialExpressionAppendVector_1 = AppendVector(
   A=CollectionParameter() {collection=/Landmass/Landscape/BlueprintBrushes/MPC/MPC_Erosion, parameterName=ResX}
   B=CollectionParameter() {collection=&MPC_Erosion, parameterName=ResY}
  )
  MaterialExpressionTextureCoordinate_1 = TextureCoordinate()
  MaterialExpressionCollectionParameter_0 = CollectionParameter() {collection=&MPC_Erosion, parameterName=GridSize}
  MaterialExpressionCollectionParameter_3 = CollectionParameter() {collection=&MPC_Erosion, parameterName=ZScale}
  MaterialExpressionAppendVector_3 = AppendVector(A=Constant() {r=1}, B=ScalarParameter() {defaultValue=1, parameterName=ShowWater})
  n1 = Custom!(
   HeightTex=@MaterialExpressionTextureObjectParameter_1
   TextureSize=@MaterialExpressionAppendVector_1
   UV=@MaterialExpressionTextureCoordinate_1
   GridSize=@MaterialExpressionCollectionParameter_0
   HeightScale=@MaterialExpressionCollectionParameter_3
   D=@MaterialExpressionAppendVector_3
  ) {
   code="//HeightTex: input Texture Object containing heights.R and water.B\r\n//TextureSize: input for size\r\n//GridSize: input float grid world size\r\n//HeightScale: input for height range scale\r\n\r\nfloat2 offsets[4] = {float2(0,1), float2(0,-1), float2(1,0), float2(-1,0)};\r\n\r\nfloat2 texelsize = 1 / TextureSize;\r\n\r\nUV = (floor(UV * TextureSize)/TextureSize) + (texelsize * 0.5);\r\n\r\nfloat2 PosYSample = HeightTex.SampleLevel(HeightTexSampler, UV + offsets[0] * texelsize, 0).rb;\r\nfloat2 NegYSample = HeightTex.SampleLevel(HeightTexSampler, UV + offsets[1] * texelsize, 0).rb;\r\nfloat2 PosXSample = HeightTex.SampleLevel(HeightTexSampler, UV + offsets[2] * texelsize, 0).rb;\r\nfloat2 NegXSample = HeightTex.SampleLevel(HeightTexSampler, UV + offsets[3] * texelsize, 0).rb;\r\n\r\nfloat3 tangent =  float3( 2*GridSize, 0, ((PosXSample.x) - (NegXSample.x))*HeightScale);\r\nfloat3 binormal = float3( 0, 2*GridSize, ((PosYSample.x) - (NegYSample.x))*HeightScale);\r\ntangent = normalize(tangent);\r\nbinormal = normalize(binormal);\r\n\r\nfloat3 normal1 = normalize(cross(tangent, binormal));\r\n\r\n//float2 normal1 = float2(-tangent.z, -binormal.z);\r\n\r\nPosYSample.x +=  PosYSample.y;\r\nNegYSample.x +=  NegYSample.y;\r\nPosXSample.x +=  PosXSample.y;\r\nNegXSample.x +=  NegXSample.y;\r\n\r\ntangent =  float3( 2*GridSize, 0, ((PosXSample.x) - (NegXSample.x))*HeightScale);\r\nbinormal = float3( 0, 2*GridSize, ((PosYSample.x) - (NegYSample.x))*HeightScale);\r\ntangent = normalize(tangent);\r\nbinormal = normalize(binormal);\r\n\r\nfloat3 normal2 = normalize(cross(tangent, binormal));\r\n\r\nreturn float4(normal1.x, normal1.y, normal2.x, normal2.y);"
   description=Compute Normal
   outputType=CMOT_Float4
   inputs [
    {inputName=HeightTex, input {expression=&self:MaterialExpressionTextureObjectParameter_1}}
    {inputName=TextureSize, input {expression=&self:MaterialExpressionAppendVector_1}}
    {inputName=UV, input {expression=&self:MaterialExpressionTextureCoordinate_1}}
    {inputName=GridSize, input {expression=&self:MaterialExpressionCollectionParameter_0}}
    {inputName=HeightScale, input {expression=&self:MaterialExpressionCollectionParameter_3}}
    {inputName=D, input {expression=&self:MaterialExpressionAppendVector_3}}
   ]
  }
  n2 = ConstantBiasScale(None=@n1)
  AppendVector(A=ComponentMask(None=@n2) {g=true, r=true}, B=TextureSampleParameter2D() {parameterName=HeightRT, texture=&Black}.R)
  ComponentMask(None=@n2) {b=true, g=true, r=true}
  Constant3Vector() {constant=<rgba 0 0.2 0.3 0>}
  MaterialExpressionTextureObjectParameter_0 = TextureObjectParameter() {parameterName=HeightRT_1, texture=&Black}
  MaterialExpressionDivide_6 = Divide(B=AppendVector(A=ScalarParameter() {defaultValue=2048, parameterName=SizeX_1}, B=ScalarParameter() {defaultValue=2048, parameterName=SizeY_1}))
  MaterialExpressionScalarParameter_26 = ScalarParameter() {defaultValue=16, parameterName=GridSize_1}
  MaterialExpressionTextureCoordinate_2 = TextureCoordinate()
  Custom!(
   HeightRT=@MaterialExpressionTextureObjectParameter_0
   TexelSize=@MaterialExpressionDivide_6
   GridSize=@MaterialExpressionScalarParameter_26
   UV=@MaterialExpressionTextureCoordinate_2
  ) {
   code="//HeightRT: input Texture Object containing heights.R and water.B\r\n//TextureSize: input for size\r\n//GridSize: input float grid world size\r\n//HeightScale: input for height range scale\r\n\r\nfloat2 offsets[4] = {float2(0,1), float2(0,0), float2(1,0), float2(0,0)};\r\n\r\nfloat2 texelsize = TexelSize;\r\n\r\n\r\nfloat PosYSample = HeightRT.SampleLevel(HeightRTSampler, UV + offsets[0] * texelsize, 0).r;\r\nfloat NegYSample = HeightRT.SampleLevel(HeightRTSampler, UV + offsets[1] * texelsize, 0).r;\r\nfloat PosXSample = HeightRT.SampleLevel(HeightRTSampler, UV + offsets[2] * texelsize, 0).r;\r\nfloat NegXSample = HeightRT.SampleLevel(HeightRTSampler, UV + offsets[3] * texelsize, 0).r;\r\n\r\nfloat3 tangent =  float3( 2*GridSize, 0, ((PosXSample) - (NegXSample)));\r\nfloat3 binormal = float3( 0, 2*GridSize, ((PosYSample) - (NegYSample)));\r\ntangent = normalize(tangent);\r\nbinormal = normalize(binormal);\r\n\r\nfloat3 normal = normalize(cross(tangent, binormal));\r\n\r\nreturn normal;"
   description=Normal
   inputs [
    {inputName=HeightRT, input {expression=&self:MaterialExpressionTextureObjectParameter_0}}
    {inputName=TexelSize, input {expression=&self:MaterialExpressionDivide_6}}
    {inputName=GridSize, input {expression=&self:MaterialExpressionScalarParameter_26}}
    {inputName=UV, input {expression=&self:MaterialExpressionTextureCoordinate_2}}
   ]
  }
  LinearInterpolate(
   A=AppendVector(A=ComponentMask(None=@n1) {g=true, r=true}, B=ComponentMask(None=@n1) {b=true})
   B=@n2
   Alpha=ScalarParameter() {parameterName=OutputLDR}
  )
  MaterialFunctionCall() {materialFunction=&LocalPosition}
  OneMinus(None=ComponentMask(None=@n2) {a=true})
  ScalarParameter() {defaultValue=1, parameterName=HeightScale}
  ScalarParameter() {defaultValue=2048, parameterName=SizeX}
  ScalarParameter() {defaultValue=2048, parameterName=SizeY}
  ScalarParameter() {defaultValue=16, parameterName=GridSize}
  n3 = ComponentMask(None=@n1) {b=true, desc=cosine}
  SquareRoot(None=OneMinus(None=Multiply(A=@n3, B=@n3))) {desc=sine}
 ]
}
