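# LSBrush_DistanceField is a prototype Landmass landscape brush material function that reads signed distance, height and mask from a jump flood render target.
# It point-samples JumpFloodRT with a UV offset, measures distance to the stored RG coordinate, and signs it by alpha and Invert.
# World Distance scales this by the larger landscape size; blue supplies Z Height and alpha the Mask, scaled by UseCaptureDepth.
# Noise-distorted UV nodes remain unconnected.
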MaterialFunction /Landmass/Landscape/OldPrototype_BP/Mats/MF/LSBrush_DistanceField
graph {nodes [
  n1 = TextureCoordinate()
  n2 = Multiply(A=@n1, B=ScalarParameter() {defaultValue=1, parameterName=distxy})
  n3 = Multiply(
   A=Subtract(A=TextureSample(UVs=@n2) {texture=/Engine/MapTemplates/TilingNoise05}.R) {constB=0.5}
   B=ScalarParameter() {defaultValue=0.1, parameterName=noise dist}
  )
  n4 = AppendVector(
   A=CollectionParameter() {collection=/Landmass/Landscape/BlueprintBrushes/MPC/MPC_Erosion, parameterName=ResX}
   B=CollectionParameter() {collection=&MPC_Erosion, parameterName=ResY}
  )
  n5 = Floor(None=Multiply(A=TextureCoordinate(), B=@n4))
  Add(A=@n3, B=Add(A=Divide(A=@n5, B=@n4), B=Divide(B=@n4) {constA=0.5}))
  Add(A=@n1, B=@n3)
  n6 = Add(
   A=MaterialFunctionCall() {materialFunction=&/PointSampleUVs}.TexelCenterUVs
   B=FunctionInput() {bUsePreviewValueAsDefault=true, inputName=UV Offset, inputType=FunctionInput_Vector2}
  )
  n7 = TextureSampleParameter2D!(UVs=@n6, Level=Constant(), Apply View MipBias) {mipValueMode=TMVM_MipLevel, parameterName=JumpFloodRT, texture=/Engine/EngineResources/Black_Low.Black}
  n8 = Subtract(A=@n6, B=ComponentMask(None=@n7.RGB) {g=true, r=true})
  n9 = LinearInterpolate(Alpha=@n7.A) {constA=1, constB=-1}
  n10 = LinearInterpolate(Alpha=ScalarParameter() {parameterName=Invert}) {constA=1, constB=-1}
  n11 = Multiply(A=Multiply(
    A=MaterialFunctionCall(Vector 2=@n8) {materialFunction=/Engine/Functions/Engine_MaterialFunctions02/Utility/VectorLength}.V2 Length
    B=Multiply(A=@n9, B=@n10)
   )) {constB=-1}
  FunctionOutput(None=@n11) {description=The distance in 0-1 UV space within the render target of the landscape., outputName=UV Distance}
  n12 = CollectionParameter() {collection=&MPC_Erosion, parameterName=TotalSizeX}
  n13 = CollectionParameter() {collection=&MPC_Erosion, parameterName=TotalSizeY}
  FunctionOutput(None=Multiply(A=@n11, B=Max(A=@n12, B=@n13))) {
   description="The distance in world space, based on a tangent of 1, aka a 45 degree angle which has a tangent or multiplier of 1.0."
   outputName=World Distance
   sortPriority=1
  }
  FunctionOutput(None=@n7.B) {
   description="The distance in world space, based on a tangent of 1, aka a 45 degree angle which has a tangent or multiplier of 1.0."
   outputName=Z Height
   sortPriority=2
  }
  FunctionOutput(None=Multiply(A=@n7.A, B=ScalarParameter() {defaultValue=1, parameterName=UseCaptureDepth})) {
   description="The distance in world space, based on a tangent of 1, aka a 45 degree angle which has a tangent or multiplier of 1.0."
   outputName=Mask
   sortPriority=3
  }
  TextureCoordinate()
 ]}
