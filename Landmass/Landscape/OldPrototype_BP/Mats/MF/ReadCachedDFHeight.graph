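# Material function in a prototype Landmass landscape setup that reads a cached distance field and height from the CachedDistanceFieldHeight texture parameter.
# It samples mip 0 at texel centers, unpacking red and green into a doubled UV Distance.
# World Distance multiplies that by the larger TotalSizeX or TotalSizeY from MPC_Erosion.
# Height unpacks the blue channel, fed into both inputs of UnpackRG8Height.
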MaterialFunction /Landmass/Landscape/OldPrototype_BP/Mats/MF/ReadCachedDFHeight
graph {nodes [
  n1 = TextureSampleParameter2D!(
   UVs=MaterialFunctionCall() {materialFunction=&/PointSampleUVs}.TexelCenterUVs
   Level=Constant()
   Apply View MipBias
  ) {mipValueMode=TMVM_MipLevel, parameterName=CachedDistanceFieldHeight, texture=/Engine/EngineResources/Black_Low.Black}
  FunctionOutput(None=MaterialFunctionCall!(RG=AppendVector(A=@n1.B, B=@n1.B)) {desc=Height, materialFunction=&/UnpackRG8Height}.Result) {outputName=Height, sortPriority=2}
  n2 = Divide(A=MaterialFunctionCall(RG=ComponentMask(None=@n1.RGB) {g=true, r=true}) {materialFunction=&UnpackRG8Height}.Result) {constB=0.5, desc=DF}
  n3 = CollectionParameter() {collection=/Landmass/Landscape/BlueprintBrushes/MPC/MPC_Erosion, parameterName=TotalSizeX}
  n4 = CollectionParameter() {collection=&MPC_Erosion, parameterName=TotalSizeY}
  FunctionOutput(None=Multiply(A=FunctionOutput(None=@n2) {outputName=UV Distance}, B=Max(A=@n3, B=@n4))) {
   description="The distance in world space, based on a tangent of 1, aka a 45 degree angle which has a tangent or multiplier of 1.0."
   outputName=World Distance
   sortPriority=1
  }
 ]}
