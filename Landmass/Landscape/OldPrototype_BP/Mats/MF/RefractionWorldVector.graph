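# Material function that computes a world-space refracted view vector from two refractive indices (air-to-water defaults), a surface normal and an incident vector defaulting to the negated camera vector.
# Refracted Vector applies the vector form of Snell's law using the index ratio and incidence cosine.
# Lerp Approximation and SingleLerp offer cheaper index-independent bends toward the normal, tuned by constants and oddly named scalar parameters.
# Several alternative derivations remain unconnected to any output.
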MaterialFunction /Landmass/Landscape/OldPrototype_BP/Mats/MF/RefractionWorldVector
graph {nodes [
  n1 = FunctionInput() {bUsePreviewValueAsDefault=true, inputName=Refractive Index 1, inputType=FunctionInput_Scalar, previewValue=<1 0 0 1>}
  n2 = FunctionInput() {bUsePreviewValueAsDefault=true, inputName=Refractive Index 2, inputType=FunctionInput_Scalar, sortPriority=1, previewValue=<1.33 0 0 1>}
  n3 = Divide(A=@n1, B=@n2)
  n4 = FunctionInput() {bUsePreviewValueAsDefault=true, inputName=Surface Normal, sortPriority=2, previewValue=<0 0 1 1>}
  n5 = FunctionInput(Preview=Multiply(A=CameraVectorWS()) {constB=-1}) {bUsePreviewValueAsDefault=true, desc=Incident Vector, inputName=Camera Vector, sortPriority=3, previewValue=<0 0 1 1>}
  n6 = CrossProduct(A=Multiply(A=@n4) {constB=-1}, B=@n5)
  n7 = CrossProduct(A=@n4, B=@n5)
  n8 = OneMinus(None=Multiply(A=Multiply(A=@n3, B=@n3), B=DotProduct(A=@n7, B=@n7)))
  Add(A=Multiply(A=@n3, B=CrossProduct(A=@n4, B=@n6)), B=Multiply(A=@n4, B=SquareRoot(None=@n8)))
  n9 = DotProduct(A=@n5, B=@n4)
  n10 = LinearInterpolate(A=Constant() {r=0.465}, B=Constant() {r=0.225}, Alpha=@n9)
  Add(A=@n5, B=Multiply(A=@n10, B=@n4))
  n11 = DotProduct(A=Multiply(A=@n5) {constB=-1}, B=@n4) {desc=cosine}
  n12 = Multiply(A=@n11, B=@n11) {desc="cos^2"}
  n13 = OneMinus(None=@n12)
  n14 = Multiply(A=@n3, B=@n3) {desc="n' ^2"}
  n15 = SquareRoot(None=OneMinus(None=Multiply(A=@n13, B=@n14)))
  n16 = Add(A=Multiply(A=@n3, B=@n11) {desc=n1/n2*cos}, B=@n15)
  FunctionOutput(None=Add(A=Multiply(A=@n5, B=@n3) {desc=n1/n2 * i}, B=Multiply(A=@n16, B=@n4))) {outputName=Refracted Vector}
  n17 = ScalarParameter() {defaultValue=0.75, parameterName=faasdfasfasdf}
  FunctionOutput(None=Divide(A=LinearInterpolate(A=@n5, B=@n4, Alpha=@n10), B=LinearInterpolate(A=@n17, Alpha=@n9) {constA=0.706})) {outputName=Lerp Approximation, sortPriority=1}
  n18 = Multiply(A=@n4)
  n19 = ScalarParameter() {defaultValue=0.75, parameterName=UDT}
  FunctionOutput(None=Add(A=@n5, B=Multiply(A=@n18, B=@n19))) {outputName=SingleLerp, sortPriority=1}
  Multiply(A=@n15) {constB=-1}
  Subtract(A=@n12)
 ]}
