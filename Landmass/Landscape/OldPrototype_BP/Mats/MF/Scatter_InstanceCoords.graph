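# Material function for a Landmass landscape prototype that maps each mesh instance's local UVs into texture coordinates across the whole landscape, likely for scattering.
# The local origin, transformed to world space, is normalized by total size and floored to a section index using MPC_Erosion parameters.
# That index steps by component resolution minus one, is added to UV1 scaled by component resolution, then divided by total resolution.
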MaterialFunction /Landmass/Landscape/OldPrototype_BP/Mats/MF/Scatter_InstanceCoords
graph {nodes [
  n1 = Multiply(
   A=TextureCoordinate() {coordinateIndex=1}
   B=CollectionParameter() {collection=/Landmass/Landscape/BlueprintBrushes/MPC/MPC_Erosion, parameterName=ComponentRes}
  )
  n2 = TransformPosition(None=Constant3Vector()) {transformType=TRANSFORMPOSSOURCE_World}
  n3 = CollectionParameter() {collection=&MPC_Erosion, parameterName=TotalSizeX}
  n4 = CollectionParameter() {collection=&MPC_Erosion, parameterName=TotalSizeY}
  n5 = Add(A=Divide(A=ComponentMask(None=@n2) {g=true, r=true}, B=AppendVector(A=@n3, B=@n4))) {constB=0.5, desc=0-1 for grid}
  n6 = AppendVector(
   A=CollectionParameter() {collection=&MPC_Erosion, parameterName=ResX}
   B=CollectionParameter() {collection=&MPC_Erosion, parameterName=ResY}
  )
  n7 = Subtract(A=CollectionParameter() {collection=&MPC_Erosion, parameterName=ComponentRes})
  n8 = Divide(A=Subtract(A=@n6), B=@n7) {desc=Sections XY}
  n9 = Multiply(A=Floor(None=Multiply(A=@n5, B=@n8)), B=@n7)
  FunctionOutput(None=Divide(A=Add(A=@n1, B=@n9), B=@n6))
 ]}
