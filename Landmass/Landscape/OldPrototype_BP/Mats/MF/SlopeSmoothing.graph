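# Material function for a prototype Landmass erosion brush that smooths slopes in a heightmap, driven by MPC_Erosion parameters.
# It samples four HeightRT neighbors, alternating diagonal and axial offsets by iteration parity, to derive an average and minimum (reading only R and B channels).
# Smoothed slopes are height-capped using a grid-scaled threshold above that minimum.
# A thresholded flat-smoothing mask and opacity blends produce the output Height.
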MaterialFunction /Landmass/Landscape/OldPrototype_BP/Mats/MF/SlopeSmoothing
graph {nodes [
  n1 = NamedRerouteUsage()
  n2 = MaterialFunctionCall(Float4=NamedRerouteUsage()) {materialFunction=&BreakOutFloat4Components}
  n3 = Min(A=@n2.R, B=@n2.R)
  n4 = Min(A=@n2.B, B=@n2.B)
  n5 = Add(
   A=NamedRerouteDeclaration(None=Min(A=@n3, B=@n4)) {name=Min Neighbor, nodeColor=<rgba 0.6172 1 0 1>}
   B=Multiply(
    A=CollectionParameter() {collection=/Landmass/Landscape/BlueprintBrushes/MPC/MPC_Erosion, parameterName=GridSize}
    B=CollectionParameter() {collection=&MPC_Erosion, parameterName=SlopeSmoothThreshold}
   )
  )
  n6 = NamedRerouteUsage()
  n7 = NamedRerouteDeclaration(None=Min(A=LinearInterpolate(A=@n5, B=@n6), B=NamedRerouteUsage())) {name=Smoothed Slopes, nodeColor=<rgba 1 0.2812 0 1>}
  n8 = NamedRerouteUsage()
  FunctionOutput(None=LinearInterpolate(
    A=NamedRerouteUsage()
    B=LinearInterpolate(A=@n1, B=@n7, Alpha=@n8)
    Alpha=CollectionParameter() {collection=&MPC_Erosion, parameterName=SlopeSmoothingOpacity}
   )) {outputName=Height}
  NamedRerouteDeclaration(None=FunctionInput(Preview=TextureCoordinate()) {bUsePreviewValueAsDefault=true, inputName=UVs, inputType=FunctionInput_Vector2}) {name=Input UVs, nodeColor=<rgba 0 1 0.5078 1>}
  NamedRerouteDeclaration(None=Fmod(A=CollectionParameter() {collection=&MPC_Erosion, parameterName=idx}, B=Constant() {r=2})) {name=Every Other Frame Mask, nodeColor=<rgba 0.4062 0 1 1>}
  n9 = CollectionParameter() {collection=&MPC_Erosion, parameterName=ResX}
  n10 = CollectionParameter() {collection=&MPC_Erosion, parameterName=ResY}
  NamedRerouteDeclaration(None=Divide(B=AppendVector(A=@n9, B=@n10))) {name=TexelSize, nodeColor=<rgba 0.9922 1 0 1>}
  n11 = NamedRerouteUsage()
  NamedRerouteDeclaration(None=FunctionInput(Preview=TextureSampleParameter2D(UVs=@n11) {parameterName=HeightRT, texture=&Black}.R) {bUsePreviewValueAsDefault=true, inputName=Height, inputType=FunctionInput_Scalar, sortPriority=1}) {name=Starting Height, nodeColor=<rgba 1 0.7969 0 1>}
  n12 = NamedRerouteUsage()
  n13 = LinearInterpolate(A=Constant2Vector() {g=-0.707, r=-0.707}, B=Constant2Vector() {r=-1}, Alpha=NamedRerouteUsage())
  n14 = MaterialFunctionCall!(Add=Add(A=NamedRerouteUsage(), B=Multiply(A=@n12, B=@n13))) {materialFunction=/Landmass/Landscape/ErosionBrush/Materials/Functions/ErosionBrush_ClampUVs}
  n15 = NamedRerouteUsage()
  n16 = LinearInterpolate(A=Constant2Vector() {g=-0.707, r=0.707}, B=Constant2Vector() {g=-1}, Alpha=NamedRerouteUsage())
  n17 = MaterialFunctionCall(Add=Add(A=NamedRerouteUsage(), B=Multiply(A=@n15, B=@n16))) {materialFunction=&ErosionBrush_ClampUVs}
  n18 = NamedRerouteUsage()
  n19 = LinearInterpolate(A=Constant2Vector() {g=0.707, r=0.707}, B=Constant2Vector() {r=1}, Alpha=NamedRerouteUsage())
  n20 = MaterialFunctionCall(Add=Add(A=NamedRerouteUsage(), B=Multiply(A=@n18, B=@n19))) {materialFunction=&ErosionBrush_ClampUVs}
  n21 = NamedRerouteUsage()
  n22 = LinearInterpolate(A=Constant2Vector() {g=0.707, r=-0.707}, B=Constant2Vector() {g=1}, Alpha=NamedRerouteUsage())
  n23 = MaterialFunctionCall(Add=Add(A=NamedRerouteUsage(), B=Multiply(A=@n21, B=@n22))) {materialFunction=&ErosionBrush_ClampUVs}
  NamedRerouteDeclaration(None=MaterialFunctionCall(
    X=TextureSampleParameter2D(UVs=@n14.Clamp) {parameterName=HeightRT, texture=&Black}.R
    Y=TextureSampleParameter2D(UVs=@n17.Clamp) {parameterName=HeightRT, texture=&Black}.R
    Z=TextureSampleParameter2D(UVs=@n20.Clamp) {parameterName=HeightRT, texture=&Black}.R
    A=TextureSampleParameter2D(UVs=@n23.Clamp) {parameterName=HeightRT, texture=&Black}.R
   ) {materialFunction=&MakeFloat4}.Result) {name=FourNeighborsV4, nodeColor=<rgba 0.2891 0 1 1>}
  n24 = DotProduct(A=NamedRerouteUsage(), B=Constant() {r=1})
  n25 = NamedRerouteUsage()
  NamedRerouteDeclaration(None=Divide(A=Add(A=@n24, B=@n25)) {constB=5}) {name=Average, nodeColor=<rgba 1 0 0.1641 1>}
  n26 = NamedRerouteUsage()
  n27 = NamedRerouteUsage()
  n28 = Subtract(
   A=Divide(
    A=Subtract(A=@n26, B=@n27)
    B=CollectionParameter() {collection=&MPC_Erosion, parameterName=GridSize}
   )
   B=CollectionParameter() {collection=&MPC_Erosion, parameterName=FlatSmoothThreshold}
  )
  NamedRerouteDeclaration(None=Saturate(None=Ceil(None=@n28))) {name=Flat Smooth Mask, nodeColor=<rgba 0.6875 0 1 1>}
 ]}
