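# SmoothIntersections is a material function in an old Landmass landscape prototype that blends inputs A and B into smooth maximum and minimum outputs.
# It derives a clamped blend factor h from (A − B) / K, where K sets the blend width and has a preview default.
# Each output lerps between the inputs by h, then adds or subtracts K·h·(1−h), the polynomial smooth-min technique.
# The ConstantBiasScale and masked texture coordinate nodes feed no output.
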MaterialFunction /Landmass/Landscape/OldPrototype_BP/Mats/MF/SmoothIntersections
graph {nodes [
  ComponentMask(None=TextureCoordinate()) {r=true}
  n1 = FunctionInput() {inputName=A}
  n2 = FunctionInput() {inputName=B, sortPriority=1}
  n3 = FunctionInput() {bUsePreviewValueAsDefault=true, inputName=K, sortPriority=2, previewValue=<1 1 1 1>}
  n4 = Divide(A=Subtract(A=@n1, B=@n2), B=@n3)
  ConstantBiasScale(None=@n4)
  n5 = Saturate(None=Add(A=Multiply(A=@n4) {constB=0.5}) {constB=0.5}) {desc=h}
  n6 = Multiply(A=Multiply(A=@n5, B=OneMinus(None=@n5)), B=@n3)
  FunctionOutput(None=Add(A=LinearInterpolate(A=@n2, B=@n1, Alpha=@n5), B=@n6)) {outputName=Smooth Max}
  FunctionOutput(None=Subtract(A=LinearInterpolate(A=@n1, B=@n2, Alpha=@n5), B=@n6)) {outputName=Smooth Min, sortPriority=1}
 ]}
