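# Material function from an old Landmass prototype that computes velocity-advected UVs, apparently for erosion simulation.
# It point-samples VelocityRT and HeightRT at texel centers, scaling velocity XY by MPC_Erosion's AdvectionErosion, deltaT and a constant divisor, then by the height texture's blue channel.
# The resulting world-space offset is divided by total size to give a UV offset, which is added to the texel-center UVs.
# All three values are outputs.
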MaterialFunction /Landmass/Landscape/OldPrototype_BP/Mats/MF/VelocityAdvect
graph {nodes [
  n1 = MaterialFunctionCall() {materialFunction=&/PointSampleUVs}
  n2 = Constant()
  n3 = CollectionParameter() {collection=/Landmass/Landscape/BlueprintBrushes/MPC/MPC_Erosion, parameterName=AdvectionErosion}
  n4 = CollectionParameter() {collection=&MPC_Erosion, parameterName=deltaT}
  n5 = Multiply(
   A=ComponentMask(None=TextureSampleParameter2D!(UVs=@n1.TexelCenterUVs, Level=@n2, Apply View MipBias) {mipValueMode=TMVM_MipLevel, parameterName=VelocityRT, texture=&Black}.RGB) {g=true, r=true}
   B=Divide(A=Multiply(A=@n3, B=@n4)) {constB=100}
  )
  n6 = TextureSampleParameter2D(UVs=@n1.TexelCenterUVs, Level=Constant()) {mipValueMode=TMVM_MipLevel, parameterName=HeightRT, texture=&Black}
  n7 = Divide(
   A=FunctionOutput(None=Multiply(A=@n5, B=@n6.B)) {outputName=OffsetWorldSpace, sortPriority=2}
   B=AppendVector(
    A=CollectionParameter() {collection=&MPC_Erosion, parameterName=TotalSizeX}
    B=CollectionParameter() {collection=&MPC_Erosion, parameterName=TotalSizeY}
   )
  )
  FunctionOutput(None=Add(A=MaterialFunctionCall() {materialFunction=&PointSampleUVs}.TexelCenterUVs, B=@n7)) {outputName=Advected UVs}
  FunctionOutput(None=@n7) {outputName=Offset, sortPriority=1}
 ]}
